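# Reusable scroll-box list widget for tech-art Blutilities, showing one configurable item widget per data object.
# Adding items creates widgets, optionally overrides button style and colours, passes data through TA_ListInterface and records a widget-to-data map.
# Mouse handlers hit-test hovered children for left and right clicks, driving single, multi and Shift-range selection.
# Dirty flags let Tick batch selection-changed and most data-changed events, while clicks broadcast through dispatchers immediately.
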WidgetBlueprint /Game/Blueprints/Blutilities/ForTechArt/Widgets/List/TA_Generic01_List
graphs [
 EventGraph {nodes [
   Event PreConstruct()
   > SetOrientation!(
    execute
    self:Scroll Box Object Reference=Get ScrollBox_List() -> (ScrollBox_List:Scroll Box Object Reference)
    NewOrientation:EOrientation Enum=Get scrollOrientation() -> (scrollOrientation:EOrientation Enum=Orient_Horizontal)
   ) -> (then)
   > SetAlwaysShowScrollbar!(
    execute
    self:Scroll Box Object Reference=Get ScrollBox_List()
    NewAlwaysShowScrollbar:bool=Get alwaysShowScrollBar() -> (alwaysShowScrollBar:bool)
   ) -> (then)
   Event Tick()
   > tick_validateSelectionChanged()
   > tick_throwDataChangedEvent()
  ]}
 OnMouseButtonDown {nodes [
   OnMouseButtonDown!() -> (then, MyGeometry:Geometry Structure, MouseEvent:"Pointer Event Structure (by ref)") {is=title K2Node_FunctionEntry}
   > Set itemHoveredOnDown(itemHoveredOnDown:User Widget Object Reference) -> (Output_Get:User Widget Object Reference)
   > n1 = For Each Loop with Break!(Exec:Exec, Array:Array of Widget Object References=GetAllChildren(self=Get ScrollBox_List()), Break:Exec=@n3.then) -> (LoopBody:Exec, Array Element:Widget Object Reference, Array Index:int, Completed:Exec)
   Return Node!(execute=@n1.Completed, ReturnValue:Event Reply Structure=Handled())
   > Branch(Condition=IsHovered(self=@n1))
   > n2 = Cast To UserWidget!(execute, Object=@n1) -> (then, CastFailed, AsUser Widget:User Widget Object Reference) {is=title K2Node_DynamicCast}
   > n3 = Set itemHoveredOnDown(itemHoveredOnDown=@n2)
   Unhandled()
  ]}
 OnMouseButtonUp {nodes [
   OnMouseButtonUp!() -> (then, MyGeometry:Geometry Structure, MouseEvent:"Pointer Event Structure (by ref)") {is=title K2Node_FunctionEntry}
   n1 = Get MouseEvent() -> (MouseEvent:"Pointer Event Structure (by ref)")
   > Branch(Condition="Equal (Key)"!(
     A:Key Structure=Get Effecting Button!(Input:"Pointer Event Structure (by ref)"=@n1) -> (ReturnValue:Key Structure) {is=title K2Node_CallFunction}
     B:Key Structure=LeftMouseButton
    ) -> (ReturnValue:bool))
   > n2 = Sequence()
   n3 = Get itemHoveredOnDown() -> (itemHoveredOnDown:User Widget Object Reference)
   > n4 = Is Valid(InputObject=@n3)
   Return Node(execute=@n4.Is Not Valid, ReturnValue=Unhandled())
   > n5 = Branch(Condition=IsHovered(self=@n3))
   > n6 = getListDataFromItem!(execute, self, itemWidget:User Widget Object Reference=Get itemHoveredOnDown()) -> (then, exists:bool, data:Object Reference)
   > n7 = Branch(Condition=@n6)
   n8 = GetInputEventFromPointerEvent!(Event:"Pointer Event Structure (by ref)"=Get MouseEvent()) -> (ReturnValue:Input Event Structure)
   > n9 = Call On Item Left Clicked!(
    execute:Exec
    self:TA Generic 01 List Object Reference
    isShiftDown:bool=Is Shift Down!(Input:"Input Event Structure (by ref)"=@n8) -> (ReturnValue:bool)
    isCtrlDown:bool=Is Control Down!(Input:"Input Event Structure (by ref)"=@n8) -> (ReturnValue:bool)
    data:Object Reference=@n6.data
    item:User Widget Object Reference=Get itemHoveredOnDown()
   ) -> (then:Exec) {is=title K2Node_CallDelegate}
   Branch(execute=[@n9.then, @n7.else], Condition=Get enableSelection() -> (enableSelection:bool))
   > n10 = Branch(Condition=ShouldAllowMultiselectMouseEvent!(Event:"Pointer Event Structure (by ref)"=Get MouseEvent()) -> (ReturnValue:bool))
   n11 = GetInputEventFromPointerEvent(Event=Get MouseEvent())
   n12 = Get lastSelectedItem() -> (lastSelectedItem:User Widget Object Reference)
   > n13 = Branch(Condition=AND Boolean(A=IsValid(Object=@n12), B=Is Shift Down(Input=@n11)))
   > n14 = getIndexOfItem!(execute, self, Content:Widget Object Reference=Get lastSelectedItem()) -> (then, index:int=0)
   > n15 = getIndexOfItem(Content=Get itemHoveredOnDown()) -> (index=0)
   > selectFromIndexToIndex!(execute, self, index1:int=@n15, index2:int=@n14) -> (then)
   n16 = Branch(
    execute=@n13.else
    Condition=Contains Item!(
     TargetSet:Set of User Widget Object References=Get selectedItems() -> (selectedItems:Set of User Widget Object References)
     ItemToFind:"User Widget Object Reference (by ref)"=Get itemHoveredOnDown()
    ) -> (ReturnValue:bool) {is=title K2Node_CallFunction}
   )
   n17 = Get selectedItems()
   > n18 = Branch(Condition="Equal (Integer)"(A=Length!(TargetSet:Set of User Widget Object References=@n17) -> (ReturnValue:int=0) {is=title K2Node_CallFunction}, B=1) {is=title K2Node_CallFunction})
   > n19 = Branch(Condition=Get allowNoSelection() -> (allowNoSelection:bool))
   setListItemSelection!(
    execute=[@n19.then, @n18.else]
    self
    inWidget:User Widget Object Reference=Get itemHoveredOnDown()
    selected:bool=false
    throwEvent:bool=true
   ) -> (then)
   setListItemSelection(execute=@n16.else, inWidget=Get itemHoveredOnDown(), selected=true, throwEvent=true)
   n20 = Length(TargetSet=Get selectedItems()) -> (ReturnValue=0)
   n21 = Get itemHoveredOnDown()
   n22 = Get selectedItems()
   n23 = Branch(execute=@n10.else, Condition=AND Boolean(A=Contains Item(TargetSet=@n22, ItemToFind=@n21), B="Equal (Integer)"(A=@n20, B=1)))
   > Branch(Condition=Get allowNoSelection())
   > clearSelection()
   setSelectedItems!(
    execute=@n23.else
    self
    items:Array of User Widget Object References=Make Array("[0]":User Widget Object Reference=Get itemHoveredOnDown()) -> (Array:Array of User Widget Object References)
   ) -> (then)
   n24 = GetInputEventFromPointerEvent(Event=Get MouseEvent())
   Branch(
    execute=@n5.else
    Condition=AND Boolean!(A=Get allowMultiSelect() -> (allowMultiSelect:bool), B=Is Shift Down(Input=@n24), C:bool=Get enableSelection()) -> (ReturnValue)
   )
   > n25 = getHoveredListItem!(execute, self, startFrom:int=0) -> (then, exists:bool, AsUser Widget:User Widget Object Reference)
   > Branch(Condition=@n25)
   > n26 = getIndexOfItem(Content=@n25.AsUser Widget) -> (index=0)
   > n27 = getIndexOfItem(Content=Get itemHoveredOnDown()) -> (index=0)
   > n28 = For Loop(FirstIndex="Min (Integer)"(A=@n26, B=@n27) -> (ReturnValue=0), LastIndex="Max (Integer)"(A=@n26, B=@n27))
   > n29 = getItemAtIndex!(execute, self, itemIndex:int=@n28) -> (then, item:User Widget Object Reference)
   > setListItemSelection(inWidget=@n29, selected=true, throwEvent=true)
   Set itemHoveredOnDown(execute=@n2.then_1)
   > Return Node(ReturnValue=Handled())
   Branch(Condition=Get enableSelection())
  ]}
 OnPreviewMouseButtonDown {nodes [
   n1 = OnPreviewMouseButtonDown!() -> (then, MyGeometry:Geometry Structure, MouseEvent:"Pointer Event Structure (by ref)") {is=title K2Node_FunctionEntry}
   n2 = Get Effecting Button(Input=@n1.MouseEvent)
   > n3 = Branch(Condition="Equal (Key)"(A=@n2, B=RightMouseButton))
   n4 = GetInputEventFromPointerEvent(Event=@n1.MouseEvent)
   > handleRightClickOnItem!(
    execute
    self
    isShiftDown:bool=Is Shift Down(Input=@n4)
    isCtrlDown:bool=Is Control Down(Input=@n4)
    Event:"Pointer Event Structure (by ref)"=@n1.MouseEvent
   ) -> (then)
   > Return Node(ReturnValue=Select(
     Option 0:Event Reply Structure=Unhandled()
     Option 1:Event Reply Structure=Handled()
     Index=Get captureMouseInputBeforeChildren() -> (captureMouseInputBeforeChildren:bool)
    ) -> (ReturnValue:Event Reply Structure))
   n5 = Branch(execute=@n3.else, Condition="Equal (Key)"(A=@n2, B=MiddleMouseButton))
   Branch(execute=@n5.else, Condition="Equal (Key)"(A=@n2, B=LeftMouseButton))
   > n6 = getHoveredListItem(startFrom=0)
   > Branch(Condition=@n6)
   n7 = GetInputEventFromPointerEvent(Event=@n1.MouseEvent)
   Is Control Down(Input=@n7)
   Is Shift Down(Input=@n7)
  ]}
 ShouldAllowMultiselectMouseEvent {nodes [
   n1 = ShouldAllowMultiselectMouseEvent!() -> (then, Event:"Pointer Event Structure (by ref)") {is=title K2Node_FunctionEntry}
   n2 = GetInputEventFromPointerEvent(Event=@n1)
   n3 = Is Control Down(Input=@n2)
   n4 = Is Shift Down(Input=@n2)
   n5 = Get multiSelectWithoutShift() -> (multiSelectWithoutShift:bool)
   > Return Node(ReturnValue:bool=AND Boolean!(A=OR Boolean!(A=@n4, B=@n3, C:bool=@n5) -> (ReturnValue), B=Get allowMultiSelect()) -> (ReturnValue))
  ]}
 ShouldAllowSelectingWidget {nodes [
   n1 = ShouldAllowSelectingWidget!() -> (then, Event:"Pointer Event Structure (by ref)", widget:"User Widget Object Reference (by ref)") {is=title K2Node_FunctionEntry}
   > n2 = Branch(Condition=Get enableSelection())
   > n3 = Branch(Condition=Contains Item(TargetSet=Get selectedItems(), ItemToFind=@n1.widget))
   > Return Node(ReturnValue=false)
   n4 = Get selectedItems()
   n5 = Branch(execute=@n3.else, Condition="integer > integer"(A=Length(TargetSet=@n4) -> (ReturnValue=0)))
   > n6 = Branch(Condition=ShouldAllowMultiselectMouseEvent(Event=@n1) {is=title K2Node_CallFunction})
   > Return Node(ReturnValue=true)
   Return Node(execute=@n6.else, ReturnValue=false)
   Return Node(execute=@n5.else, ReturnValue=true)
   Return Node(execute=@n2.else, ReturnValue=false)
  ]}
 addItem {nodes [
   addItem!() -> (then, dataObject:Object Reference) {is=title K2Node_FunctionEntry}
   > n1 = createNewListItem!(execute, self) -> (then, success:bool, newWidget:User Widget Object Reference)
   > Set l_newItem(l_newItem:User Widget Object Reference=@n1.newWidget) -> (Output_Get:User Widget Object Reference)
   > n2 = Sequence!(execute) -> (then_0, then_1, then_2:Exec, then_3:Exec, then_4:Exec)
   > n3 = Cast To TA_BaseEditorButtonWidget!(execute, Object=Get l_newItem() -> (l_newItem:User Widget Object Reference)) -> (then, CastFailed, AsTA Base Editor Button Widget:TA Base Editor Button Widget Object Reference) {is=title K2Node_DynamicCast}
   > Branch(Condition=Get overrideStyle() -> (overrideStyle:bool))
   > Set ButtonStyle_Default!(
    execute
    ButtonStyle_Default:Button Style Structure=Get ButtonStyle_Default() -> (ButtonStyle_Default:Button Style Structure)
    self:TA Base Editor Button Widget Object Reference=@n3
   ) -> (then, Output_Get:Button Style Structure)
   > Set ButtonStyle_Selected!(
    execute
    ButtonStyle_Selected:Button Style Structure=Get ButtonStyle_Selected() -> (ButtonStyle_Selected:Button Style Structure)
    self:TA Base Editor Button Widget Object Reference=@n3
   ) -> (then, Output_Get:Button Style Structure)
   n4 = Cast To TA_BaseEditorButtonWidget(execute=@n2.then_1, Object=Get l_newItem())
   > Branch(Condition=Get overrideColor() -> (overrideColor:bool))
   > Set labelColor_Default!(
    execute
    labelColor_Default:Linear Color Structure=Get labelColor_Default() -> (labelColor_Default:Linear Color Structure)
    self:TA Base Editor Button Widget Object Reference=@n4
   ) -> (then, Output_Get:Linear Color Structure)
   > Set labelColor_Selected!(
    execute
    labelColor_Selected:Linear Color Structure=Get labelColor_Selected() -> (labelColor_Selected:Linear Color Structure)
    self:TA Base Editor Button Widget Object Reference=@n4
   ) -> (then, Output_Get:Linear Color Structure)
   > Set buttonColor_Default!(
    execute
    buttonColor_Default:Linear Color Structure=Get buttonColor_Default() -> (buttonColor_Default:Linear Color Structure)
    self:TA Base Editor Button Widget Object Reference=@n4
   ) -> (then, Output_Get:Linear Color Structure)
   > Set buttonColor_Selected!(
    execute
    buttonColor_Selected:Linear Color Structure=Get buttonColor_Selected() -> (buttonColor_Selected:Linear Color Structure)
    self:TA Base Editor Button Widget Object Reference=@n4
   ) -> (then, Output_Get:Linear Color Structure)
   n5 = Branch(
    execute=@n2.then_2
    Condition=Does Object Implement Interface!(
     TestObject:Object Reference=Get l_newItem()
     Interface:Interface Interface=/Game/Blueprints/Blutilities/ForTechArt/Widgets/List/Data/TA_ListInterface.TA_ListInterface_C
    ) -> (ReturnValue:bool)
   )
   > setOwningList!(execute:Exec, self:Object Reference=Get l_newItem(), list:User Widget Object Reference=Self-Reference()) -> (then:Exec) {is=title K2Node_Message}
   > n6 = setListItemData!(execute:Exec, self:Object Reference=Get l_newItem(), listItemData:Object Reference=Get dataObject() -> (dataObject:Object Reference)) -> (then:Exec) {is=title K2Node_Message}
   n7 = AddChild(execute=[@n5.else, @n6.then], self=Get ScrollBox_List(), Content=Get l_newItem())
   > n8 = Cast To ScrollBoxSlot!(execute, Object=@n7) -> (then, CastFailed, AsScroll Box Slot:Scroll Box Slot Object Reference) {is=title K2Node_DynamicCast}
   n9 = Get itemPadding() -> (itemPadding:double)
   > SetPadding!(
    execute
    self:Scroll Box Slot Object Reference=@n8
    InPadding:Margin Structure=Select!(
     Orient_Horizontal:Margin Structure=Make Margin(Left=0, Top=0, Right=@n9, Bottom=0)
     Orient_Vertical:Margin Structure=Make Margin(Left=0, Top=0, Right=0, Bottom=@n9)
     Index:EOrientation Enum=Get scrollOrientation() -> (scrollOrientation=Orient_Horizontal)
    ) -> (ReturnValue:Margin Structure)
   ) -> (then)
   Set hasDataChanged(execute=@n2.then_3, hasDataChanged:bool=true) -> (Output_Get:bool)
   n10 = Get l_newItem()
   n11 = Get dataObject()
   Add!(
    execute=@n2.then_4
    TargetMap:Map of User Widget Object References to Object References=Get listItemDataMap() -> (listItemDataMap:Map of User Widget Object References to Object References)
    Key:"User Widget Object Reference (by ref)"=@n10
    Value:"Object Reference (by ref)"=@n11
   ) -> (then) {is=title K2Node_CallFunction}
   > Return Node!(execute, newItem:User Widget Object Reference=@n10, newData:Object Reference=@n11)
  ]}
 addItemGeneric {nodes [
   n1 = addItemGeneric!() -> (then, displayString:String) {is=title K2Node_FunctionEntry}
   > n2 = Construct Object from Class(Class=/Game/Blueprints/Blutilities/ForTechArt/Data/TA_GenericListItemData.TA_GenericListItemData_C, self=Self-Reference()) -> (ReturnValue:TA Generic List Item Data Object Reference)
   > Set displayString!(execute, displayString:String=@n1, self:TA Generic List Item Data Object Reference=@n2) -> (then, Output_Get:String)
   > n3 = addItem!(execute, self, dataObject:Object Reference=@n2) -> (then, newItem:User Widget Object Reference, newData:Object Reference) {is=title K2Node_CallFunction}
   > Return Node!(execute, widget:User Widget Object Reference=@n3, data:Object Reference=@n3.newData)
  ]}
 addItems {nodes [
   n1 = addItems!() -> (then, dataObjects:Array of Object References) {is=title K2Node_FunctionEntry}
   > n2 = For Each Loop(Array:Array of Object References=@n1) -> (Array Element:Object Reference)
   > addItem(dataObject=@n2)
  ]}
 clearItems {nodes [
   clearItems() {is=title K2Node_FunctionEntry}
   > ClearChildren(self=Get ScrollBox_List())
   > Clear!(execute, TargetMap:Map of User Widget Object References to Object References=Get listItemDataMap()) -> (then) {is=title K2Node_CallFunction}
   > Call On Items Cleared!(execute:Exec, self:TA Generic 01 List Object Reference) -> (then:Exec) {is=title K2Node_CallDelegate}
   > Set hasDataChanged(hasDataChanged=true)
  ]}
 clearSelection {nodes [
   clearSelection() {is=title K2Node_FunctionEntry}
   > Set selectionIsInvalid(selectionIsInvalid:bool=true) -> (Output_Get:bool)
   > n1 = To Array!(execute, A:Set of User Widget Object References=Get selectedItems()) -> (then, Result:Array of User Widget Object References)
   > n2 = For Each Loop(Array:Array of User Widget Object References=@n1) -> (Array Element:User Widget Object Reference)
   Clear!(execute=@n2.Completed, TargetSet:Set of User Widget Object References=Get selectedItems()) -> (then)
   > Set lastSelectedItem(lastSelectedItem:User Widget Object Reference) -> (Output_Get:User Widget Object Reference)
   setListItemSelection!(execute:Exec=@n2.LoopBody, self:Object Reference=@n2, selected:bool=false) -> (then:Exec) {is=title K2Node_Message}
  ]}
 clearSelection_NoEvent {nodes [
   clearSelection_NoEvent() {is=title K2Node_FunctionEntry}
   > Set l_isSelectionInvalid(l_isSelectionInvalid:bool=Get selectionIsInvalid() -> (selectionIsInvalid:bool)) -> (Output_Get:bool)
   > clearSelection() {is=title K2Node_CallFunction}
   > Set selectionIsInvalid(selectionIsInvalid=Get l_isSelectionInvalid() -> (l_isSelectionInvalid:bool))
  ]}
 createNewListItem {nodes [
   createNewListItem() {is=title K2Node_FunctionEntry}
   n1 = Get listItemClass() -> (listItemClass:User Widget Class Reference)
   > n2 = Branch(Condition=IsValidClass!(Class:Object Class Reference=@n1) -> (ReturnValue:bool))
   > n3 = Create Widget!(execute:Exec, Class:User Widget Class Reference=@n1, OwningPlayer:Player Controller Object Reference) -> (then:Exec, ReturnValue:User Widget Object Reference)
   > Return Node!(execute, success:bool=true, newWidget:User Widget Object Reference=@n3)
   Return Node(execute=@n2.else, success=false)
  ]}
 getHoveredListItem {nodes [
   n1 = getHoveredListItem!() -> (then, startFrom:int) {is=title K2Node_FunctionEntry}
   n2 = Get ScrollBox_List()
   > n3 = For Loop(
    FirstIndex=@n1
    LastIndex=int - int(A=GetChildrenCount!(self:Panel Widget Object Reference=@n2) -> (ReturnValue:int=0), B=1) {is=title K2Node_CallFunction}
   )
   Return Node!(execute=@n3.Completed, exists:bool=false, AsUser Widget:User Widget Object Reference)
   > n4 = Cast To UserWidget(Object=GetChildAt!(self:Panel Widget Object Reference=Get ScrollBox_List(), Index:int=@n3) -> (ReturnValue:Widget Object Reference))
   > Branch(Condition=IsHovered(self=@n4))
   > Return Node(exists=true, AsUser Widget=@n4)
  ]}
 getIndexOfItem {nodes [
   n1 = getIndexOfItem!() -> (then, Content:Widget Object Reference) {is=title K2Node_FunctionEntry}
   > Return Node!(
    execute
    index:int=GetChildIndex!(self:Panel Widget Object Reference=Get ScrollBox_List(), Content:Widget Object Reference=@n1) -> (ReturnValue:int=0)
   )
  ]}
 getIsItemSelected {nodes [
   n1 = getIsItemSelected!() -> (then, item:User Widget Object Reference) {is=title K2Node_FunctionEntry}
   > Return Node!(execute, isSelected:bool=Contains Item(TargetSet=Get selectedItems(), ItemToFind=@n1))
  ]}
 getItemAtIndex {nodes [
   n1 = getItemAtIndex!() -> (then, itemIndex:int) {is=title K2Node_FunctionEntry}
   > n2 = Cast To UserWidget(Object=GetChildAt(self=Get ScrollBox_List(), Index=@n1))
   > Return Node!(execute, item:User Widget Object Reference=@n2)
  ]}
 getListData {nodes [
   getListData() {is=title K2Node_FunctionEntry}
   > n1 = Values!(execute, TargetMap:Map of User Widget Object References to Object References=Get listItemDataMap()) -> (then, Values:Array of Object References)
   > Return Node!(execute, data:Array of Object References=@n1)
  ]}
 getListDataFromItem {nodes [
   n1 = getListDataFromItem!() -> (then, itemWidget:User Widget Object Reference) {is=title K2Node_FunctionEntry}
   n2 = Find(TargetMap:Map of User Widget Object References to Object References=Get listItemDataMap(), Key:"User Widget Object Reference (by ref)"=@n1) -> (Value:Object Reference)
   > Return Node!(execute, exists:bool=@n2.ReturnValue, data:Object Reference=@n2)
  ]}
 getListItems {nodes [
   getListItems() {is=title K2Node_FunctionEntry}
   > n1 = Keys!(execute, TargetMap:Map of User Widget Object References to Object References=Get listItemDataMap()) -> (then, Keys:Array of User Widget Object References)
   > Return Node!(execute, items:Array of User Widget Object References=@n1)
  ]}
 getSelectedData {nodes [
   getSelectedData() {is=title K2Node_FunctionEntry}
   > n1 = To Array(A=Get selectedItems())
   > n2 = For Each Loop(Array=@n1)
   Return Node!(execute=@n2.Completed, selectedData:Array of Object References=Get l_data() -> (l_data:Array of Object References))
   n3 = Find(TargetMap=Get listItemDataMap(), Key=@n2)
   > Branch(Condition=@n3.ReturnValue)
   > Add(TargetArray:Array of Object References=Get l_data(), NewItem:"Object Reference (by ref)"=@n3) {is=title K2Node_CallArrayFunction}
  ]}
 getSelectedIndex {nodes [
   getSelectedIndex() {is=title K2Node_FunctionEntry}
   n1 = Get selectedItems()
   > n2 = Branch(Condition="integer > integer"(A=Length(TargetSet=@n1) -> (ReturnValue=0), B=0) {is=title K2Node_CallFunction})
   > n3 = To Array(A=Get selectedItems())
   > n4 = getIndexOfItem(Content="Get (a copy)"(
     Array:Array of User Widget Object References=@n3
     Dimension 1=Last Index!(TargetArray:Array of User Widget Object References=@n3) -> (ReturnValue=0)
    ) -> (Output:User Widget Object Reference)) -> (index=0) {is=title K2Node_CallFunction}
   > Return Node!(execute, index:int=@n4)
   Return Node(execute=@n2.else, index=-1)
  ]}
 getSelectedIndices {nodes [
   getSelectedIndices() {is=title K2Node_FunctionEntry}
   n1 = Get selectedItems()
   > n2 = Branch(Condition="integer > integer"(A=Length(TargetSet=@n1) -> (ReturnValue=0), B=0))
   > n3 = To Array(A=Get selectedItems())
   > n4 = For Each Loop(Array=@n3)
   Return Node(execute=@n4.Completed, index:Array of Integers=Get l_indices() -> (l_indices:Array of Integers))
   > n5 = Keys(TargetMap=Get listItemDataMap())
   > Add(
    TargetArray:Array of Integers=Get l_indices()
    NewItem:"Integer (by ref)"=Find Item!(TargetArray:Array of User Widget Object References=@n5, ItemToFind:"User Widget Object Reference (by ref)"=@n4) -> (ReturnValue)
   )
   Return Node(execute=@n2.else, index=Get l_indices())
  ]}
 getSelectedItems {nodes [
   getSelectedItems() {is=title K2Node_FunctionEntry}
   > n1 = To Array(A=Get selectedItems())
   > n2 = getSelectedData!(execute, self) -> (then, selectedData:Array of Object References) {is=title K2Node_CallFunction}
   > Return Node!(
    execute
    isAnySelected:bool="integer > integer"(A=Length(TargetArray:Array of User Widget Object References=@n1) {is=title K2Node_CallArrayFunction}, B=0)
    items:Array of User Widget Object References=@n1
    selectedData:Array of Object References=@n2
   )
  ]}
 handleLeftClickOnItem {nodes [
   n1 = handleLeftClickOnItem!() -> (then, isShiftDown:bool=~, isCtrlDown:bool=~) {is=title K2Node_FunctionEntry}
   > n2 = getHoveredListItem(startFrom=0) {is=title K2Node_CallFunction}
   > Branch(Condition=@n2)
   n3 = Find(TargetMap=Get listItemDataMap(), Key=@n2.AsUser Widget)
   > Branch(Condition=@n3.ReturnValue)
   > Call On Item Left Clicked(isShiftDown=@n1, isCtrlDown=@n1.isCtrlDown, data=@n3, item=@n2.AsUser Widget)
  ]}
 handleMiddleClickOnItem {nodes [
   n1 = handleMiddleClickOnItem!() -> (then, isShiftDown:bool=~, isCtrlDown:bool=~) {is=title K2Node_FunctionEntry}
   > n2 = getHoveredListItem(startFrom=0)
   > Branch(Condition=@n2)
   n3 = Find(TargetMap=Get listItemDataMap(), Key=@n2.AsUser Widget)
   > Branch(Condition=@n3.ReturnValue)
   > Call On Item Middle Clicked!(
    execute:Exec
    self:TA Generic 01 List Object Reference
    isShiftDown:bool=@n1
    isCtrlDown:bool=@n1.isCtrlDown
    Data:Object Reference=@n3
    item:User Widget Object Reference=@n2.AsUser Widget
   ) -> (then:Exec) {is=title K2Node_CallDelegate}
  ]}
 handleRightClickOnItem {nodes [
   n1 = handleRightClickOnItem!() -> (then, isShiftDown:bool=~, isCtrlDown:bool=~, Event:"Pointer Event Structure (by ref)") {is=title K2Node_FunctionEntry}
   > n2 = getHoveredListItem(startFrom=0)
   > Branch(Condition=@n2)
   n3 = Find(TargetMap=Get listItemDataMap(), Key=@n2.AsUser Widget)
   > Branch(Condition=@n3.ReturnValue)
   > n4 = Sequence!(execute) -> (then_0, then_1)
   > n5 = Branch(Condition=ShouldAllowSelectingWidget!(Event:"Pointer Event Structure (by ref)"=@n1.Event, widget:"User Widget Object Reference (by ref)"=@n2.AsUser Widget) -> (ReturnValue:bool) {is=title K2Node_CallFunction})
   > setListItemSelection(inWidget=@n2.AsUser Widget, selected=true, throwEvent=true) {is=title K2Node_CallFunction}
   n6 = getIsItemSelected!(execute=@n5.else, self, item:User Widget Object Reference=@n2.AsUser Widget) -> (then, isSelected:bool) {is=title K2Node_CallFunction}
   > n7 = Branch(Condition=@n6)
   setSelectedItems(execute=@n7.else, items=Make Array("[0]"=@n2.AsUser Widget))
   Call On Item Right Clicked!(
    execute:Exec=@n4.then_1
    self:TA Generic 01 List Object Reference
    isShiftDown:bool=@n1
    isCtrlDown:bool=@n1.isCtrlDown
    data:Object Reference=@n3
    item:User Widget Object Reference=@n2.AsUser Widget
   ) -> (then:Exec) {is=title K2Node_CallDelegate}
  ]}
 onDataChanged {nodes [onDataChanged!() -> (then, list:TA Generic 01 List Object Reference, data:Array of Object References) {is=title K2Node_FunctionEntry}]}
 onItemLeftClicked {nodes [onItemLeftClicked!() -> (then, isShiftDown:bool=~, isCtrlDown:bool=~, data:Object Reference, item:User Widget Object Reference) {is=title K2Node_FunctionEntry}]}
 onItemMiddleClicked {nodes [onItemMiddleClicked!() -> (then, isShiftDown:bool=~, isCtrlDown:bool=~, Data:Object Reference, item:User Widget Object Reference) {is=title K2Node_FunctionEntry}]}
 onItemRightClicked {nodes [onItemRightClicked!() -> (then, isShiftDown:bool=~, isCtrlDown:bool=~, data:Object Reference, item:User Widget Object Reference) {is=title K2Node_FunctionEntry}]}
 onItemSelectionChanged {nodes [onItemSelectionChanged!() -> (then, selectedData:Array of Object References, selectedWidgets:Array of User Widget Object References, selectedIndices:Array of Integers) {is=title K2Node_FunctionEntry}]}
 onItemsCleared {nodes [onItemsCleared() {is=title K2Node_FunctionEntry}]}
 removeItemAtIndex {nodes [
   n1 = removeItemAtIndex!() -> (then, index:int) {is=title K2Node_FunctionEntry}
   n2 = Get ScrollBox_List()
   > Branch(Condition="integer < integer"!(A:int=@n1, B:int=GetChildrenCount(self=@n2) -> (ReturnValue=0)) -> (ReturnValue:bool) {is=title K2Node_CallFunction})
   n3 = Get index() -> (index:int=0)
   n4 = Get ScrollBox_List()
   > Remove!(
    execute
    TargetMap:Map of User Widget Object References to Object References=Get listItemDataMap()
    Key:"User Widget Object Reference (by ref)"=Cast To UserWidget!(Object=GetChildAt(self=@n4, Index=@n3)) -> (AsUser Widget, bSuccess:bool=~)
   ) -> (then, ReturnValue:bool)
   > RemoveChildAt!(execute, self:Panel Widget Object Reference=Get ScrollBox_List(), Index:int=Get index() -> (index=0)) -> (then, ReturnValue:bool)
  ]}
 removeItems {nodes [
   n1 = removeItems!() -> (then, itemWidgets:Array of User Widget Object References) {is=title K2Node_FunctionEntry}
   > n2 = Sequence()
   > n3 = For Each Loop(Array=@n1)
   Set lastSelectedItem(execute=@n3.Completed)
   Set l_currentItem(execute=@n3.LoopBody, l_currentItem:User Widget Object Reference=@n3) -> (Output_Get:User Widget Object Reference)
   > Set l_object(l_object:Object Reference=Find(TargetMap=Get listItemDataMap(), Key=Get l_currentItem() -> (l_currentItem:User Widget Object Reference))) -> (Output_Get:Object Reference)
   > Remove(TargetMap=Get listItemDataMap(), Key=Get l_currentItem())
   > RemoveChild(self=Get ScrollBox_List(), Content=Get l_currentItem())
   > n4 = Remove!(
    execute
    TargetSet:Set of User Widget Object References=Get selectedItems()
    Item:"User Widget Object Reference (by ref)"=Get l_currentItem()
   ) -> (then, ReturnValue)
   > Branch(Condition=@n4)
   > Set selectionIsInvalid(selectionIsInvalid=true)
   n5 = getListData!(execute=@n2.then_1, self) -> (then, data:Array of Object References) {is=title K2Node_CallFunction}
   > Call On Data Changed!(
    execute:Exec
    self:TA Generic 01 List Object Reference
    list:TA Generic 01 List Object Reference=Self-Reference()
    data:Array of Object References=@n5
   ) -> (then:Exec) {is=title K2Node_CallDelegate}
   For Each Loop(Array=Get itemWidgets() -> (itemWidgets:Array of User Widget Object References))
  ]}
 selectFromIndexToIndex {nodes [
   n1 = selectFromIndexToIndex!() -> (then, index1:int, index2:int) {is=title K2Node_FunctionEntry}
   > n2 = For Loop(FirstIndex="Min (Integer)"(A=@n1, B=@n1.index2) -> (ReturnValue=0), LastIndex="Max (Integer)"(A=@n1, B=@n1.index2))
   > n3 = getItemAtIndex(itemIndex=@n2) {is=title K2Node_CallFunction}
   > setListItemSelection(inWidget=@n3, selected=true, throwEvent=true)
  ]}
 setIndexSelected {nodes [
   n1 = setIndexSelected!() -> (then, index:int, selected:bool=~, throwEvent:bool=true) {is=title K2Node_FunctionEntry}
   n2 = Get ScrollBox_List()
   > Branch(Condition="integer < integer"(A=@n1, B=GetChildrenCount(self=@n2) -> (ReturnValue=0)))
   > n3 = Cast To UserWidget!(execute:Exec, Object=GetChildAt(self=Get ScrollBox_List(), Index=@n1)) -> (then:Exec, CastFailed:Exec, AsUser Widget)
   > setListItemSelection(inWidget=@n3, selected=@n1.selected, throwEvent=@n1.throwEvent)
  ]}
 setItems {nodes [
   n1 = setItems!() -> (then, dataObjects:Array of Object References) {is=title K2Node_FunctionEntry}
   > clearItems() {is=title K2Node_CallFunction}
   > n2 = For Each Loop(Array:Array of Object References=@n1) -> (Array Element:Object Reference)
   > addItem(dataObject=@n2)
  ]}
 setListItemSelection {nodes [
   n1 = setListItemSelection!() -> (then, inWidget:User Widget Object Reference, selected:bool=~, throwEvent:bool=true) {is=title K2Node_FunctionEntry}
   > Set selectionIsInvalid(selectionIsInvalid=@n1.throwEvent)
   > n2 = Branch(Condition=Get allowMultiSelect())
   n3 = Branch(
    execute=[@n7.then, @n2.then]
    Condition=Does Object Implement Interface(TestObject=@n1, Interface=&TA_ListInterface_C)
   )
   > n4 = setListItemSelection(self=Get inWidget() -> (inWidget:User Widget Object Reference), selected=Get selected() -> (selected:bool)) {is=title K2Node_Message}
   n5 = Branch(execute=[@n3.else, @n4.then], Condition=Get selected())
   n6 = Get inWidget()
   > Add!(execute, TargetSet:Set of User Widget Object References=Get selectedItems(), NewItem:"User Widget Object Reference (by ref)"=@n6) -> (then) {is=title K2Node_CallFunction}
   > Set lastSelectedItem(lastSelectedItem=@n6)
   Remove(execute=@n5.else, TargetSet=Get selectedItems(), Item=Get inWidget())
   > Branch(Condition="Equal (Object)"(A=Get inWidget(), B=Get lastSelectedItem()) {is=title K2Node_CallFunction})
   > Set lastSelectedItem()
   n7 = clearSelection_NoEvent(execute=@n2.else) {is=title K2Node_CallFunction}
  ]}
 setSelectedItems {nodes [
   n1 = setSelectedItems!() -> (then, items:Array of User Widget Object References) {is=title K2Node_FunctionEntry}
   > Set l_inItems(l_inItems:Array of User Widget Object References=@n1) -> (Output_Get:Array of User Widget Object References)
   > Set selectionIsInvalid(selectionIsInvalid=true)
   > n2 = To Array(A=Get selectedItems())
   > n3 = Set l_allItems(l_allItems:Array of User Widget Object References=Get l_inItems() -> (l_inItems:Array of User Widget Object References)) -> (Output_Get:Array of User Widget Object References)
   > Append Array!(execute, TargetArray:Array of User Widget Object References=@n3, SourceArray:Array of User Widget Object References=@n2) -> (then)
   > n4 = For Each Loop(Array:Array of User Widget Object References=Get l_allItems() -> (l_allItems:Array of User Widget Object References)) -> (Array Element:User Widget Object Reference)
   n5 = Get selectedItems()
   Clear(execute=@n4.Completed, TargetSet=@n5)
   > Add Items!(execute, TargetSet:Set of User Widget Object References=@n5, NewItems:Array of User Widget Object References=Get l_inItems()) -> (then)
   n6 = Branch(
    execute=@n4.LoopBody
    Condition=Contains Item(TargetArray:Array of User Widget Object References=Get l_inItems(), ItemToFind:"User Widget Object Reference (by ref)"=@n4) {is=title K2Node_CallArrayFunction}
   )
   > n7 = Branch(Condition=Contains Item(TargetSet=Get selectedItems(), ItemToFind=@n4) {is=title K2Node_CallFunction})
   Set lastSelectedItem(execute=[@n7.then, @n8.then], lastSelectedItem=@n4)
   n8 = setListItemSelection(execute=@n7.else, self=@n4, selected=true)
   setListItemSelection(execute=@n6.else, self=@n4, selected=false)
  ]}
 tick_throwDataChangedEvent {nodes [
   tick_throwDataChangedEvent() {is=title K2Node_FunctionEntry}
   > Branch(Condition=Get hasDataChanged() -> (hasDataChanged:bool))
   > Set hasDataChanged(hasDataChanged=false)
   > n1 = getListData()
   > Call On Data Changed(list=Self-Reference(), data=@n1)
  ]}
 tick_validateSelectionChanged {nodes [
   tick_validateSelectionChanged() {is=title K2Node_FunctionEntry}
   > Branch(Condition=Get selectionIsInvalid())
   > Set selectionIsInvalid(selectionIsInvalid=false)
   > n1 = getSelectedItems!(execute, self) -> (then, isAnySelected:bool, items:Array of User Widget Object References, selectedData:Array of Object References) {is=title K2Node_CallFunction}
   > n2 = getSelectedIndices!(execute, self) -> (then, index:Array of Integers) {is=title K2Node_CallFunction}
   > Call On Item Selection Changed!(
    execute:Exec
    self:TA Generic 01 List Object Reference
    selectedData:Array of Object References=@n1.selectedData
    selectedWidgets:Array of User Widget Object References=@n1.items
    selectedIndices:Array of Integers=@n2
   ) -> (then:Exec) {is=title K2Node_CallDelegate}
  ]}
]
variables [
 allowMultiSelect {type=boolean, default=true}
 allowNoSelection {type=boolean}
 alwaysShowScrollBar {type=boolean}
 buttonColor_Default {type=LinearColor, default=<rgba 1 1 1 1>}
 buttonColor_Selected {type=LinearColor, default=<rgba 1 1 1 1>}
 ButtonStyle_Default {
  type=ButtonStyle
  default {
   clickedSlateSound {resourceObject=null}
   disabled {
    bIsDynamicallyLoaded=false
    drawAs=NoDrawType
    imageType=NoImage
    mirroring=NoMirror
    resourceName=None
    resourceObject=null
    tiling=NoTile
    imageSize=<0 0>
    margin=<ltrb 0 0 0 0>
    outlineSettings {
     bUseBrushTransparency=false
     roundingType=HalfHeightRadius
     width=0
     color {colorUseRule=UseColor_Specified, specifiedColor=<rgba 0 0 0 0>}
     cornerRadii=<0 0 0 1>
    }
    tintColor {colorUseRule=UseColor_Specified, specifiedColor=<rgba 0.4167 0.4167 0.4167 1>}
    uVRegion {bIsValid=false, max=<0 0>, min=<0 0>}
   }
   disabledForeground {colorUseRule=UseColor_Foreground, specifiedColor=<rgba 1 0 1 1>}
   hovered {
    bIsDynamicallyLoaded=false
    drawAs=Image
    imageType=NoImage
    mirroring=NoMirror
    resourceName=None
    resourceObject=null
    tiling=NoTile
    imageSize=<32 32>
    margin=<ltrb 0 0 0 0>
    outlineSettings {
     bUseBrushTransparency=false
     roundingType=HalfHeightRadius
     width=0
     color {colorUseRule=UseColor_Specified, specifiedColor=<rgba 0 0 0 0>}
     cornerRadii=<0 0 0 1>
    }
    tintColor {colorUseRule=UseColor_Specified, specifiedColor=<rgba 0.8542 0.8542 0.8542 1>}
    uVRegion {bIsValid=false, max=<0 0>, min=<0 0>}
   }
   hoveredForeground {colorUseRule=UseColor_Foreground, specifiedColor=<rgba 1 0 1 1>}
   hoveredSlateSound {resourceObject=null}
   normal {
    bIsDynamicallyLoaded=false
    drawAs=Image
    imageType=NoImage
    mirroring=NoMirror
    resourceName=None
    resourceObject=null
    tiling=NoTile
    imageSize=<32 32>
    margin=<ltrb 0 0 0 0>
    outlineSettings {
     bUseBrushTransparency=false
     roundingType=HalfHeightRadius
     width=0
     color {colorUseRule=UseColor_Specified, specifiedColor=<rgba 0 0 0 0>}
     cornerRadii=<0 0 0 1>
    }
    tintColor {colorUseRule=UseColor_Specified, specifiedColor=<rgba 0.5457 0.5457 0.5457 1>}
    uVRegion {bIsValid=false, max=<0 0>, min=<0 0>}
   }
   normalForeground {colorUseRule=UseColor_Foreground, specifiedColor=<rgba 1 0 1 1>}
   normalPadding=<ltrb 0 0 0 0>
   pressed {
    bIsDynamicallyLoaded=false
    drawAs=Image
    imageType=NoImage
    mirroring=NoMirror
    resourceName=None
    resourceObject=null
    tiling=NoTile
    imageSize=<32 32>
    margin=<ltrb 0 0 0 0>
    outlineSettings {
     bUseBrushTransparency=false
     roundingType=HalfHeightRadius
     width=0
     color {colorUseRule=UseColor_Specified, specifiedColor=<rgba 0 0 0 0>}
     cornerRadii=<0 0 0 1>
    }
    tintColor {colorUseRule=UseColor_Specified, specifiedColor=<rgba 0.2396 0.2396 0.2396 1>}
    uVRegion {bIsValid=false, max=<0 0>, min=<0 0>}
   }
   pressedForeground {colorUseRule=UseColor_Foreground, specifiedColor=<rgba 1 0 1 1>}
   pressedPadding=<ltrb 2 2 2 2>
   pressedSlateSound {resourceObject=null}
  }
 }
 ButtonStyle_Selected {
  type=ButtonStyle
  default {
   clickedSlateSound {resourceObject=null}
   disabled {
    bIsDynamicallyLoaded=false
    drawAs=NoDrawType
    imageType=NoImage
    mirroring=NoMirror
    resourceName=None
    resourceObject=null
    tiling=NoTile
    imageSize=<0 0>
    margin=<ltrb 0 0 0 0>
    outlineSettings {
     bUseBrushTransparency=false
     roundingType=HalfHeightRadius
     width=0
     color {colorUseRule=UseColor_Specified, specifiedColor=<rgba 0 0 0 0>}
     cornerRadii=<0 0 0 1>
    }
    tintColor {colorUseRule=UseColor_Specified, specifiedColor=<rgba 0.4167 0.4167 0.4167 1>}
    uVRegion {bIsValid=false, max=<0 0>, min=<0 0>}
   }
   disabledForeground {colorUseRule=UseColor_Foreground, specifiedColor=<rgba 1 0 1 1>}
   hovered {
    bIsDynamicallyLoaded=false
    drawAs=Image
    imageType=NoImage
    mirroring=NoMirror
    resourceName=None
    resourceObject=null
    tiling=NoTile
    imageSize=<32 32>
    margin=<ltrb 0 0 0 0>
    outlineSettings {
     bUseBrushTransparency=false
     roundingType=HalfHeightRadius
     width=0
     color {colorUseRule=UseColor_Specified, specifiedColor=<rgba 0 0 0 0>}
     cornerRadii=<0 0 0 1>
    }
    tintColor {colorUseRule=UseColor_Specified, specifiedColor=<rgba 0.8542 0.8542 0.8542 1>}
    uVRegion {bIsValid=false, max=<0 0>, min=<0 0>}
   }
   hoveredForeground {colorUseRule=UseColor_Foreground, specifiedColor=<rgba 1 0 1 1>}
   hoveredSlateSound {resourceObject=null}
   normal {
    bIsDynamicallyLoaded=false
    drawAs=Image
    imageType=NoImage
    mirroring=NoMirror
    resourceName=None
    resourceObject=null
    tiling=NoTile
    imageSize=<32 32>
    margin=<ltrb 0 0 0 0>
    outlineSettings {
     bUseBrushTransparency=false
     roundingType=HalfHeightRadius
     width=0
     color {colorUseRule=UseColor_Specified, specifiedColor=<rgba 0 0 0 0>}
     cornerRadii=<0 0 0 1>
    }
    tintColor {colorUseRule=UseColor_Specified, specifiedColor=<rgba 0.5457 0.5457 0.5457 1>}
    uVRegion {bIsValid=false, max=<0 0>, min=<0 0>}
   }
   normalForeground {colorUseRule=UseColor_Foreground, specifiedColor=<rgba 1 0 1 1>}
   normalPadding=<ltrb 0 0 0 0>
   pressed {
    bIsDynamicallyLoaded=false
    drawAs=Image
    imageType=NoImage
    mirroring=NoMirror
    resourceName=None
    resourceObject=null
    tiling=NoTile
    imageSize=<32 32>
    margin=<ltrb 0 0 0 0>
    outlineSettings {
     bUseBrushTransparency=false
     roundingType=HalfHeightRadius
     width=0
     color {colorUseRule=UseColor_Specified, specifiedColor=<rgba 0 0 0 0>}
     cornerRadii=<0 0 0 1>
    }
    tintColor {colorUseRule=UseColor_Specified, specifiedColor=<rgba 0.2396 0.2396 0.2396 1>}
    uVRegion {bIsValid=false, max=<0 0>, min=<0 0>}
   }
   pressedForeground {colorUseRule=UseColor_Foreground, specifiedColor=<rgba 1 0 1 1>}
   pressedPadding=<ltrb 2 2 2 2>
   pressedSlateSound {resourceObject=null}
  }
 }
 captureMouseInputBeforeChildren {type=boolean}
 enableSelection {type=boolean, default=true}
 hasDataChanged {type=boolean, default=true}
 itemHoveredOnDown {type=UserWidget, default=null}
 itemPadding {type=number, default=3}
 labelColor_Default {type=LinearColor, default=<rgba 0.0208 0.0208 0.0208 1>}
 labelColor_Selected {type=LinearColor, default=<rgba 0.8906 0.8906 0.8906 1>}
 lastSelectedItem {type=UserWidget, default=null}
 listItemClass {type=Class, default=/Game/Blueprints/Blutilities/ForTechArt/Widgets/Buttons/TA_Generic01_ListItem.TA_Generic01_ListItem_C}
 listItemDataMap {default {}}
 multiSelectWithoutShift {type=boolean}
 onDataChanged {}
 onItemLeftClicked {}
 onItemMiddleClicked {}
 onItemRightClicked {}
 onItemsCleared {}
 onItemSelectionChanged {}
 overrideColor {type=boolean}
 overrideStyle {type=boolean}
 RightClickChangesSelection {type=boolean, default=true}
 scrollOrientation {type=EOrientation, default=Orient_Vertical}
 selectedItems {type="array<UserWidget>", default []}
 selectionIsInvalid {type=boolean}
]
widgets {
 info {inheritedWidgetCount=0, parentClass=&UserWidget, rootWidgetClass=&ScrollBox, widgetCount=1}
 widgets [ScrollBox_List {widgetClassPath=&ScrollBox, bIsVariable=true, widget=ScrollBox_List, properties {scrollbarThickness=<14 9>}}]
}
